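# Niagara idle effect for the RavenQuill pickaxe: shed feathers, a soft glow and wing fragments.
# FeatherL bursts one or two tumbling mesh feathers on randomly delayed loops, which drift down under drag and fade.
# GlowLow spawns pulsing blue-to-violet sprites that also drive lights through their particle color.
# Wings_Fragments streams small meshes that stretch, shift color over life and animate dynamic material parameters as they fade.
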NiagaraSystem /BRCosmetics/Weapons/FORT_Melee/Pickaxe_RavenQuill/FX/NS_Pickaxe_Ravenquill_Idle
dependencies {
 usedDataInterfaces [&NiagaraDataInterfaceVectorCurve, &NiagaraDataInterfaceCurve]
 usedDynamicInputs []
 usedModules [&SystemState]
}
emitters {
 FeatherL {
  emitterUpdateScript [
   &EmitterState(
    Life Cycle Mode=::NewEnumerator1
    Loop Delay=&RandomRangeFloat(Minimum=0.5, Maximum=2, Evaluation Type=::NewEnumerator1)
    Loop Duration=&RandomRangeFloat(Minimum=1, Maximum=1.5, Evaluation Type=::NewEnumerator1)
    UseLoopDelay=true
   )
   &SpawnBurst_Instantaneous(
    Spawn Count=&RandomRangeInt(Minimum=1, Maximum=2)
    Use Spawn Probability=false
    UseLoopCountLimit=false
   )
  ]
  particleSpawnScript [
   /SetVariables(Particles.MeshRotationEuler:Vector3f=&RandomRangeVector(Minimum=<-0.1 0.5 0.225>, Maximum=<0.1 0.5 0.275>, Evaluation Type=::NewEnumerator1)) {isSetParameters=true}
   /SetVariables(Particles.MeshOrientation=/CascadeToNiagaraConverter/NiagaraScripts/DynamicInputScripts/MakeQuatFromEuler(Euler:Vector3f=$Particles.MeshRotationEuler, Scale:NiagaraFloat=360)) {isSetParameters=true}
   /SetVariables(Particles.Acceleration:Vector3f=&RandomRangeVector(Minimum=<0 0 -50>, Maximum=<0 0 -100>, Evaluation Type=::NewEnumerator1)) {isSetParameters=true}
   /SetVariables(Particles.MeshRotationRateBase:Vector3f=&RandomRangeVector(Minimum=<0 0 0>, Maximum=<1 1 1>, Evaluation Type=::NewEnumerator1)) {isSetParameters=true}
   &InitializeParticle(
    Lifetime=&RandomRangeFloat(Minimum=0.5, Maximum=1.75, Evaluation Type=::NewEnumerator1)
    Mesh Scale=/Niagara/DynamicInputs/CascadeConversion/SplitVector(
     Double Component Vector Channels:ECascadeNiagaraTwoVectorChannels=::NewEnumerator0
     Double Component:NiagaraFloat=&RandomRangeFloat(Minimum=0.15, Maximum=0.25, Evaluation Type=::NewEnumerator1)
     Single Component:NiagaraFloat=&RandomRangeFloat(Minimum=1, Maximum=0.1, Evaluation Type=::NewEnumerator1)
    )
    Mesh Scale Mode=::NewEnumerator4
    UsePositionOffset=false
    Write Material Random=true
   )
   &AddVelocity(Velocity=&RandomRangeVector(Minimum=<0 -5 0>, Maximum=<0 50 0>, Evaluation Type=::NewEnumerator1))
   &AddVelocity(
    Cone Angle=&RandomRangeFloat(Minimum=0, Maximum=0.3, Evaluation Type=::NewEnumerator1)
    Cone Angle Mode:ENiagara_AngleInput=::NewEnumerator1
    Cone Axis=<0 1 0>
    Module.UseConeFalloff=false
    Module.UseVelDistribution=false
    Override Seed=false
    Velocity Mode=::NewEnumerator2
    Velocity Speed=&RandomRangeFloat(Minimum=1, Maximum=200, Evaluation Type=::NewEnumerator1)
   ) {module=AddVelocity001}
   &ShapeLocation(
    Cylinder Height=50
    Cylinder Radius=6
    MaskSpawnGroup=false
    Override Seed=false
    Rotation Mode=::NewEnumerator2
    Shape Primitive=::NewEnumerator1
    Surface Only=false
    UseOwnerScale=true
    Yaw / Pitch / Roll=<0 0 0.25>
   )
   &DynamicMaterialParameters(
    Index 0 Param 1=&RandomRangeFloat(Minimum=0, Maximum=0.1, Evaluation Type=::NewEnumerator1)
    Index 0 Param 4=&RandomRangeFloat(Minimum=0, Maximum=0.5, Evaluation Type=::NewEnumerator1)
    Module.Param0WriteEnabled=true
    Module.Param1WriteEnabled=false
    Module.Param2WriteEnabled=false
    Module.Param3WriteEnabled=true
   )
   &Color(Color=&MakeLinearColorFromVectorAndFloat("Vector (RGB)"=<1 1 1>, "Float (Alpha)"=1))
  ]
  particleUpdateScript [
   /SetVariables(Particles.MeshRotationRate:Vector3f=$Particles.MeshRotationRateBase) {isSetParameters=true}
   /SetVariables(Particles.MeshRotationRate:Vector3f=&Multiply_Vector(
     A=$Particles.MeshRotationRate
     B=&VectorFromCurve(VectorCurve=DataInterface() {propertyValues {XCurve {keys [<at 0 1>,<at 1 0.3>]}, YCurve {keys [<at 0 1>,<at 1 0.3>]}, ZCurve {keys [<at 0 1>,<at 1 0.3>]}}})
    )) {isSetParameters=true}
   &ParticleState()
   &ScaleMeshSize(Scale Factor=&VectorFromFloat(Value=&FloatFromCurve(FloatCurve=DataInterface() {propertyValues {Curve {keys [
          {arriveTangent=27.05, interpMode=RCIM_Cubic, leaveTangent=27.05, tangentMode=RCTM_User, time=0, value=0}
          {arriveTangent=5.6481, interpMode=RCIM_Cubic, leaveTangent=5.6481, time=0.0306, value=1}
          <at 0.1771 1 RCIM_Cubic>
          <at 1 1 RCIM_Cubic>
         ]}}})))
   &AccelerationForce(Acceleration=$Particles.Acceleration)
   &Drag(
    Drag=&FloatFromCurve(FloatCurve=DataInterface() {propertyValues {Curve {keys [<at 0 2>,<at 0.2 2>]}}})
    Use Linear Drag=true
    Use Rotational Drag=true
   )
   &SolveForcesAndVelocity(Clamp Velocity=false, Limit Acceleration=false)
   &DynamicMaterialParameters(
    Index 0 Param 2=&FloatFromCurve(FloatCurve=DataInterface() {propertyValues {Curve {keys [<at 0.6 1 RCIM_Cubic>,<at 1 0 RCIM_Cubic>]}}})
    Index 0 Param 3=&FloatFromCurve(FloatCurve=DataInterface() {propertyValues {Curve {keys [<at 0 0>,<at 0.5 20>]}}})
    Module.Param0WriteEnabled=false
    Module.Param1WriteEnabled=true
    Module.Param2WriteEnabled=true
    Module.Param3WriteEnabled=false
   ) {module=DynamicMaterialParameters001}
   &ScaleColor(
    Scale Alpha=&FloatFromCurve(FloatCurve=DataInterface() {propertyValues {Curve {keys [<at 0 1>,<at 0.6 1>,<at 1 0>]}}})
    Scale RGB=&VectorFromCurve(VectorCurve=DataInterface() {propertyValues {XCurve {keys [<at 0 0.6>,<at 0.25 0.5>]}, YCurve {keys [<at 0 0.6>,<at 0.25 0.5>]}, ZCurve {keys [<at 0 0.6>,<at 0.25 0.5>]}}})
    ScaleA=true
    ScaleRGB=true
   )
   /CascadeToNiagaraConverter/NiagaraScripts/ModuleScripts/UpdateMeshOrientationByEuler(RotationRate:Vector3f=$Particles.MeshRotationRate)
  ]
  rendererClasses [&NiagaraMeshRendererProperties]
  renderers [{rendererClass=&NiagaraMeshRendererProperties, rendererIndex=0}]
 }
 GlowLow {
  emitterUpdateScript [
   &EmitterState(Life Cycle Mode=::NewEnumerator1, Loop Duration=1, UseLoopDelay=true)
   &SpawnBurst_Instantaneous(Use Spawn Probability=false, UseLoopCountLimit=false)
   &SpawnRate(SpawnRate=1.3, Use Spawn Probability=false) {module=SpawnRate001}
  ]
  particleSpawnScript [
   &InitializeParticle(
    Lifetime=&RandomRangeFloat(Minimum=1, Maximum=1.5, Evaluation Type=::NewEnumerator1)
    Sprite Size=<75 150>
    Sprite Size Mode=::NewEnumerator4
    UsePositionOffset=false
    Write Material Random=true
   )
   &Color(Color=&MakeLinearColorFromVectorAndFloat("Vector (RGB)"=<0.4582 0.6571 1>, "Float (Alpha)"=1))
   /CascadeToNiagaraConverter/NiagaraScripts/ModuleScripts/AddVectorToPosition(Vector:Vector3f=&TransformVector(Vector=<0 0 5>, Destination Space=::Simulation))
   &AddVelocity(Velocity=<0 -10 0>)
  ]
  particleUpdateScript [
   &ParticleState()
   &ScaleColor(
    Scale Alpha=&FloatFromCurve(
     FloatCurve=DataInterface() {propertyValues {Curve {keys [<at 0 0>,<at 0.3 1>,<at 1 0>]}}}
     Scale Curve=&Multiply_Float(
      A=1
      B=&FloatFromCurve(
       FloatCurve=DataInterface() {propertyValues {Curve {defaultValue=1, keys [{tangentMode=RCTM_User, time=0, value=0}, {tangentMode=RCTM_User, time=0.25, value=1}]}}}
       CurveIndex=$System.Age
      )
     )
    )
    Scale RGB=&VectorFromCurve(VectorCurve=DataInterface() {propertyValues {
       XCurve {keys [<at 0 0.2187>,<at 0.5 0.3298>,<at 1 0.4301>]}
       YCurve {keys [<at 0 0.3859>,<at 0.5 0.2087>,<at 1 0.1808>]}
       ZCurve {keys [<at 0 1>,<at 0.5 1>,<at 1 1>]}
      }})
    ScaleA=true
    ScaleRGB=true
   )
   &DynamicMaterialParameters(
    Index 0 Param 1=0.8
    Index 0 Param 2=2
    Index 0 Param 3=1
    Index 0 Param 4=1
    Module.Param0WriteEnabled=true
    Module.Param1WriteEnabled=true
    Module.Param2WriteEnabled=true
    Module.Param3WriteEnabled=true
   )
   &CameraOffset(Camera Offset Amount=7)
   &SolveForcesAndVelocity(Clamp Velocity=false, Limit Acceleration=false)
   &FN_Light_Attributes(
    Color=$Particles.Color
    Create Color Attribute=true
    Create Exponent Attribute=true
    Create Radius Attribute=true
    Exponent=3
    Radius=75
    Scale Color=&FloatFromCurve(
     FloatCurve=DataInterface() {propertyValues {Curve {
        defaultValue=1
        keys [
         {interpMode=RCIM_Cubic, tangentMode=RCTM_User, time=0, value=0}
         {interpMode=RCIM_Cubic, tangentMode=RCTM_User, time=0.5, value=1}
         {tangentMode=RCTM_User, time=1, value=0}
        ]
       }}}
     Scale Curve=5
    )
   )
  ]
  rendererClasses [&NiagaraSpriteRendererProperties, &NiagaraLightRendererProperties]
  renderers [
   {rendererClass=&NiagaraSpriteRendererProperties}
   {rendererClass=&NiagaraLightRendererProperties, rendererIndex=1}
  ]
 }
 Wings_Fragments {
  emitterUpdateScript [
   &EmitterState(Life Cycle Mode=::NewEnumerator1, Loop Duration=1, UseLoopDelay=true)
   &SpawnRate(SpawnRate=4, Use Spawn Probability=false) {module=SpawnRate001}
  ]
  particleSpawnScript [
   /SetVariables(Particles.ShapeLocationOffset:Vector3f=&RandomRangeVector(Minimum=<5 0 9>, Maximum=<1 -20 9>, Evaluation Type=::NewEnumerator1)) {isSetParameters=true}
   /SetVariables(Particles.ShapeLocationOffsetSpace:ENiagaraCoordinateSpace=::Local) {isSetParameters=true}
   /SetVariables(Particles.Acceleration:Vector3f=<0 5 -25>) {isSetParameters=true}
   &InitializeParticle(
    Mesh Scale=&RandomRangeVector(Minimum=<0.3 0.05 1>, Maximum=<0.5 0.03 1>, Evaluation Type=::NewEnumerator1)
    Mesh Scale Mode=::NewEnumerator4
    UsePositionOffset=false
    Write Material Random=true
   )
   &AddVelocity(
    Cone Angle=0
    Cone Angle Mode:ENiagara_AngleInput=::NewEnumerator1
    Cone Axis=<0 1 0>
    Module.UseConeFalloff=false
    Module.UseVelDistribution=false
    Override Seed=false
    Velocity Mode=::NewEnumerator2
    Velocity Speed=&RandomRangeFloat(Minimum=10, Maximum=30, Evaluation Type=::NewEnumerator1)
   )
   &ShapeLocation(
    Cylinder Height=1
    Cylinder Radius=1
    MaskSpawnGroup=false
    Offset=$Particles.ShapeLocationOffset
    Offset Coordinate Space=$Particles.ShapeLocationOffsetSpace
    Offset Mode=::NewEnumerator0
    Override Seed=false
    Rotation Mode=::NewEnumerator2
    Shape Primitive=::NewEnumerator1
    Surface Only=false
    UseOwnerScale=true
    Yaw / Pitch / Roll=<0 0 0.25>
   )
   &Color(Color=&MakeLinearColorFromVectorAndFloat(
     "Vector (RGB)"=&RandomRangeVector(Minimum=<0 0 0>, Maximum=<1 1 1>, Evaluation Type=::NewEnumerator1)
     "Float (Alpha)"=30
    ))
   &DynamicMaterialParameters(
    Index 0 Param 2=&RandomRangeFloat(Minimum=30, Maximum=50, Evaluation Type=::NewEnumerator1)
    Module.Param0WriteEnabled=false
    Module.Param1WriteEnabled=true
    Module.Param2WriteEnabled=false
    Module.Param3WriteEnabled=false
   ) {module=DynamicMaterialParameters001}
  ]
  particleUpdateScript [
   &ParticleState()
   &ScaleMeshSize(Scale Factor=&VectorFromCurve(VectorCurve=DataInterface() {propertyValues {XCurve {keys [<at 0 1>,<at 1 3>]}, YCurve {keys [<at 0 3>,<at 1 10>]}, ZCurve {keys [<at 0 1>,<at 1 1>]}}}))
   &AccelerationForce(Acceleration=$Particles.Acceleration)
   &SolveForcesAndVelocity(Clamp Velocity=false, Limit Acceleration=false)
   &ScaleColor(
    Scale Alpha=&FloatFromCurve(FloatCurve=DataInterface() {propertyValues {Curve {keys [
         {arriveTangent=0.2972, interpMode=RCIM_Cubic, leaveTangent=0.2972, tangentMode=RCTM_User, time=0, value=-0.0092}
         {arriveTangent=5.6523, interpMode=RCIM_Cubic, leaveTangent=5.6523, time=0.0891, value=0.7573}
         <at 0.271 1.5228 RCIM_Cubic>
         {arriveTangent=-2.5809, interpMode=RCIM_Cubic, leaveTangent=-2.5809, time=0.6259, value=0.7574}
         {arriveTangent=-2.1791, interpMode=RCIM_Cubic, leaveTangent=-2.1791, time=0.7543, value=0.2756}
         {arriveTangent=-1.1217, interpMode=RCIM_Cubic, leaveTangent=-1.1217, time=0.9211, value=0.1142}
         <at 1 0 RCIM_Cubic>
        ]}}})
    Scale RGB=&VectorFromCurve(VectorCurve=DataInterface() {propertyValues {
       XCurve {keys [<at 0.0015 1 RCIM_Cubic>,<at 0.5 1 RCIM_Cubic>,<at 0.5619 1 RCIM_Cubic>,<at 1 1 RCIM_Cubic>]}
       YCurve {keys [
         <at 0.0015 22.0935 RCIM_Cubic>
         {arriveTangent=-39.1079, interpMode=RCIM_Cubic, leaveTangent=-39.1079, time=0.5, value=5}
         {arriveTangent=-10, interpMode=RCIM_Cubic, leaveTangent=-10, time=0.5619, value=0.1782}
         <at 1 0 RCIM_Cubic>
        ]}
       ZCurve {keys [
         <at 0.0015 0 RCIM_Cubic>
         {arriveTangent=8.9225, interpMode=RCIM_Cubic, leaveTangent=8.9225, time=0.5, value=5}
         {arriveTangent=490, interpMode=RCIM_Cubic, leaveTangent=490, time=0.5619, value=5}
         <at 1 250 RCIM_Cubic>
        ]}
      }})
    ScaleA=true
    ScaleRGB=true
   )
   &DynamicMaterialParameters(
    Index 0 Param 1=&FloatFromCurve(FloatCurve=DataInterface() {propertyValues {Curve {keys [<at 0 5>,<at 1 0>]}}})
    Index 0 Param 3=&FloatFromCurve(FloatCurve=DataInterface() {propertyValues {Curve {keys [<at 0 3>,<at 1 12>]}}})
    Index 0 Param 4=&FloatFromCurve(FloatCurve=DataInterface() {propertyValues {Curve {keys [
         {arriveTangent=-0.0169, interpMode=RCIM_Cubic, leaveTangent=-0.0169, tangentMode=RCTM_User, time=0.0123, value=0.25}
         {arriveTangent=-1.0347, interpMode=RCIM_Cubic, leaveTangent=-1.0347, time=0.8212, value=0.2237}
         {arriveTangent=-6.8435, interpMode=RCIM_Cubic, leaveTangent=-6.8435, time=0.9123, value=-0.6812}
         <at 1 -1 RCIM_Cubic>
        ]}}})
    Module.Param0WriteEnabled=true
    Module.Param1WriteEnabled=false
    Module.Param2WriteEnabled=true
    Module.Param3WriteEnabled=true
   )
  ]
  rendererClasses [&NiagaraMeshRendererProperties]
  renderers [{rendererClass=&NiagaraMeshRendererProperties, rendererIndex=0}]
 }
}
summary {
 systemName=NS_Pickaxe_Ravenquill_Idle
 emitters [
  {emitterName=Wings_Fragments, rendererClasses [&NiagaraMeshRendererProperties]}
  {emitterName=FeatherL, rendererClasses [&NiagaraMeshRendererProperties]}
  {emitterName=GlowLow, rendererClasses [&NiagaraSpriteRendererProperties, &NiagaraLightRendererProperties]}
 ]
 userVariables []
}
