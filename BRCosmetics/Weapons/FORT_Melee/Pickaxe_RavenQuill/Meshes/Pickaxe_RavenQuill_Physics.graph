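# Physics asset for the RavenQuill two-handed pickaxe cosmetic, giving simple collision and joints to its handle and two dangling straps.
# A box body sits on the handle bone, and each dyn_strap bone gets an identically sized, offset and rotated thin box.
# Each strap is constrained directly to the handle with twist and swing2 locked, so it pivots hinge-like on swing1 alone.
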PhysicsAsset /BRCosmetics/Weapons/FORT_Melee/Pickaxe_RavenQuill/Meshes/Pickaxe_RavenQuill_Physics
bodies {
 Melee_TwoHanded_Handle [{shapeType=Box, extentX=7.5, extentY=7.5, extentZ=7.5}]
 dyn_strap1 [{shapeType=Box, extentX=1.01, extentY=5.0725, extentZ=12.135, center=<-3.6159 -4.3217 -0.5504>, rotation=<rot 0 155.979 -80.7347>}]
 dyn_strap2 [{shapeType=Box, extentX=1.01, extentY=5.0725, extentZ=12.135, center=<-3.6159 -4.3217 -0.5504>, rotation=<rot 0 155.979 -80.7347>}]
}
constraints [
 {
  bone1Name=dyn_strap1
  bone2Name=Melee_TwoHanded_Handle
  swing1LimitDegrees=30
  swing2LimitDegrees=5
  swing2Motion=Locked
  twistLimitDegrees=45
  twistMotion=Locked
 }
 {
  bone1Name=dyn_strap2
  bone2Name=Melee_TwoHanded_Handle
  swing1LimitDegrees=30
  swing2LimitDegrees=5
  swing2Motion=Locked
  twistLimitDegrees=45
  twistMotion=Locked
 }
]
