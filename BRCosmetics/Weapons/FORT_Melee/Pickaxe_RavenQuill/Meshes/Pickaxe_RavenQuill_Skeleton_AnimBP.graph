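# Animation Blueprint for the RavenQuill pickaxe skeleton that adds secondary physics motion, including sway on its plume bones.
# It copies the attached parent's pose, then chains three spring-driven, limit-constrained Anim Dynamics nodes on dyn_plume2 to dyn_plume4, blended at partial alpha and LOD-limited.
# A RigidBody node then simulates the pickaxe's physics asset relative to b_AR_Root with overridden gravity and damped component motion.
# The event graph holds no logic.
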AnimBlueprint /BRCosmetics/Weapons/FORT_Melee/Pickaxe_RavenQuill/Meshes/Pickaxe_RavenQuill_Skeleton_AnimBP
target_skeleton=&/Pickaxe_RavenQuill_Skeleton
graphs [
 AnimGraph {nodes [
   n1 = Anim Dynamics - Bone: dyn_plume2(ComponentPose=Local To Component(LocalPose=Copy Pose From Mesh() {bUseAttachedParent=true}), Alpha=0.33, ExternalForce="0,0,0") {
    angularSpringConstant=325
    bAngularSpring=true
    bGravityOverrideInSimSpace=true
    bLinearSpring=true
    lODThreshold=1
    linearSpringConstant=320
    alphaBoolBlend {blendOption=Linear}
    boundBone {boneName=dyn_plume2}
    componentAppliedLinearAccClamp=<1000 1000 1000>
    componentLinearAccScale=<1 1 1>
    componentLinearVelScale=<20 20 20>
    physicsBodyDefinitions [{
      sphereCollisionRadius=0
      boundBone {boneName=dyn_plume2}
      boxExtents=<30 30 30>
      constraintSetup {
       angularTargetAxis=AxisY
       angularLimitsMax=<15 0.01 15>
       angularLimitsMin=<-51 -0.01 -15>
       angularTarget=<0 0 0>
       linearAxesMax=<2.5 2 2>
       linearAxesMin=<-2.5 -2 -2>
      }
      localJointOffset=<0 0 0>
     }]
   }
   n2 = Anim Dynamics - Bone: dyn_plume3(ComponentPose=@n1, Alpha=0.33, ExternalForce="0,0,0") {
    angularSpringConstant=325
    bAngularSpring=true
    bGravityOverrideInSimSpace=true
    bLinearSpring=true
    lODThreshold=1
    linearSpringConstant=200
    alphaBoolBlend {blendOption=Linear}
    boundBone {boneName=dyn_plume3}
    componentAppliedLinearAccClamp=<1000 1000 1000>
    componentLinearAccScale=<1 1 1>
    componentLinearVelScale=<20 20 20>
    physicsBodyDefinitions [{
      sphereCollisionRadius=0
      boundBone {boneName=dyn_plume3}
      boxExtents=<5 5 5>
      constraintSetup {
       angularTargetAxis=AxisY
       twistAxis=AxisZ
       angularLimitsMax=<15 0.01 15>
       angularLimitsMin=<-15 -0.01 -15>
       angularTarget=<0 0 0>
       linearAxesMax=<10 1 10>
       linearAxesMin=<-10 -1 -10>
      }
      localJointOffset=<0 0 0>
     }]
   }
   n3 = Anim Dynamics - Bone: dyn_plume4(ComponentPose=@n2, Alpha=0.33, ExternalForce="0,0,0") {
    angularSpringConstant=325
    bAngularSpring=true
    bGravityOverrideInSimSpace=true
    bLinearSpring=true
    lODThreshold=1
    linearSpringConstant=200
    alphaBoolBlend {blendOption=Linear}
    boundBone {boneName=dyn_plume4}
    componentAppliedLinearAccClamp=<1000 1000 1000>
    componentLinearAccScale=<1 1 1>
    componentLinearVelScale=<20 20 20>
    physicsBodyDefinitions [{
      sphereCollisionRadius=0
      boundBone {boneName=dyn_plume4}
      boxExtents=<3 3 30>
      constraintSetup {
       angularTargetAxis=AxisY
       angularLimitsMax=<15 0.01 15>
       angularLimitsMin=<-47.4 -0.01 -15>
       angularTarget=<0 0 0>
       linearAxesMax=<10 1 10>
       linearAxesMin=<-10 -1 -20>
      }
      localJointOffset=<0 0 0>
     }]
   }
   n4 = RigidBody(ComponentPose=@n3) {
    bClampLinearTranslationLimitToRefPose=true
    bOverrideWorldGravity=true
    overridePhysicsAsset=&/Pickaxe_RavenQuill_Physics
    simulationSpace=BaseBoneSpace
    alphaBoolBlend {blendOption=Linear}
    baseBoneRef {boneName=b_AR_Root}
    componentAppliedLinearAccClamp=<6000 6000 6000>
    componentLinearAccScale=<0.25 0.25 0.25>
    componentLinearVelScale=<0.25 0.25 0.25>
    overrideWorldGravity {z=-980}
   }
   Output Pose(Result=Component To Local(ComponentPose=@n4)) {layerGroup=None}
  ]}
 EventGraph {nodes [Event BlueprintUpdateAnimation()]}
]
variables []
